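# Sound cue for the impact sound of the fancy tomato toy when it hits something.
# A random node picks between two hit wave variations with equal weight.
# It uses the shared toy-hit attenuation and is routed to the emote music sound class rather than an effects class.
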SoundCue /FNGameplayCues/Sounds/Fort_Impact_Sounds/Toys/Toy_TomatoFancy_Hit_Cue
root=Random_1
nodes [
 {id=Random_1, children [WavePlayer_0,WavePlayer_3], properties {weights [1,1]}}
 {id=WavePlayer_0, properties {soundWaveAssetPtr=/Game/Sounds/Fort_Impact_Sounds/Toys/Toy_TomatoFancy_Hit_01}}
 {id=WavePlayer_3, properties {soundWaveAssetPtr=/Game/Sounds/Fort_Impact_Sounds/Toys/Toy_TomatoFancy_Hit_02}}
]
properties {
 attenuationSettings=/Game/Sounds/Attenuation/Fort_ToyHit_Att
 maxDistance=2800
 soundClassObject=&Fort_Music_Emote
}
