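# Sound cue for the impact sound of the tomato toy when it hits something.
# A Random node picks with equal weight between two hit wave variations.
# It uses the shared toy-hit attenuation settings and is routed to the music emote sound class rather than an effects class.
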SoundCue /FNGameplayCues/Sounds/Fort_Impact_Sounds/Toys/Toy_Tomato_Hit_Cue
root=Random_0
nodes [
 {id=Random_0, children [WavePlayer_1,WavePlayer_2], properties {weights [1,1]}}
 {id=WavePlayer_1, properties {soundWaveAssetPtr=/Game/Sounds/Fort_Impact_Sounds/Toys/Toy_Tomato_Hit_01}}
 {id=WavePlayer_2, properties {soundWaveAssetPtr=/Game/Sounds/Fort_Impact_Sounds/Toys/Toy_Tomato_Hit_02}}
]
properties {
 attenuationSettings=/Game/Sounds/Attenuation/Fort_ToyHit_Att
 maxDistance=2800
 soundClassObject=&Fort_Music_Emote
}
